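# Physics asset for secondary motion of dangling parts on the M_MED_Flappy male medium player body cosmetic: "butter" pieces, collars, leg laces and leg straps.
# Each dyn bone gets a capsule body constrained only to its own tiny _MSTR anchor capsule, forming independent pairs rather than chains.
# Collars and straps lock twist and one swing axis, butter pieces lock only one swing, and laces lock neither.
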PhysicsAsset /BRCosmetics/Characters/Player/Male/Medium/Bodies/M_MED_Flappy/Meshes/M_MED_Flappy_Dyn
bodies {
 dyn_butter_l_01 [{shapeType=Capsule, length=8.375, radius=2.25, center=<0.5439 0.0001 2.4812>, rotation=<rot 90 0 0>}]
 dyn_butter_l_01_MSTR [{shapeType=Capsule, length=0.505, radius=0.505}]
 dyn_butter_l_02 [{shapeType=Capsule, length=8.375, radius=2.25, center=<0.5439 0.0001 2.4812>, rotation=<rot 90 0 0>}]
 dyn_butter_l_02_MSTR [{shapeType=Capsule, length=0.505, radius=0.505}]
 dyn_butter_l_03 [{shapeType=Capsule, length=8.375, radius=2.25, center=<0.5439 0.0001 2.4812>, rotation=<rot 90 0 0>}]
 dyn_butter_l_03_MSTR [{shapeType=Capsule, length=0.505, radius=0.505}]
 dyn_butter_l_04 [{shapeType=Capsule, length=8.375, radius=2.25, center=<0.5439 0.0001 2.4812>, rotation=<rot 90 0 0>}]
 dyn_butter_l_04_MSTR [{shapeType=Capsule, length=0.505, radius=0.505}]
 dyn_butter_r_01 [{shapeType=Capsule, length=8.375, radius=2.25, center=<0.5439 0.0001 2.4812>, rotation=<rot 90 0 0>}]
 dyn_butter_r_01_MSTR [{shapeType=Capsule, length=0.505, radius=0.505}]
 dyn_butter_r_02 [{shapeType=Capsule, length=8.375, radius=2.25, center=<0.5439 0.0001 2.4812>, rotation=<rot 90 0 0>}]
 dyn_butter_r_02_MSTR [{shapeType=Capsule, length=0.505, radius=0.505}]
 dyn_butter_r_03 [{shapeType=Capsule, length=8.375, radius=2.25, center=<0.5439 0.0001 2.4812>, rotation=<rot 90 0 0>}]
 dyn_butter_r_03_MSTR [{shapeType=Capsule, length=0.505, radius=0.505}]
 dyn_butter_r_04 [{shapeType=Capsule, length=8.375, radius=2.25, center=<0.5439 0.0001 2.4812>, rotation=<rot 90 0 0>}]
 dyn_butter_r_04_MSTR [{shapeType=Capsule, length=0.505, radius=0.505}]
 dyn_collar_l_01 [{shapeType=Capsule, length=5, radius=2, center=<3.2237 0 0>, rotation=<rot 90 0 0>}]
 dyn_collar_l_01_MSTR [{shapeType=Capsule, length=0.505, radius=0.505}]
 dyn_collar_r_01 [{shapeType=Capsule, length=5, radius=2, center=<3.2237 0 0>, rotation=<rot 90 0 0>}]
 dyn_collar_r_01_MSTR [{shapeType=Capsule, length=0.505, radius=0.505}]
 dyn_leg_laces_l [{shapeType=Capsule, length=5, radius=2, center=<3.0121 0 0>, rotation=<rot 90 0 0>}]
 dyn_leg_laces_l_MSTR [{shapeType=Capsule, length=0.505, radius=0.505}]
 dyn_leg_strap_l_01 [{shapeType=Capsule, length=5, radius=2, center=<3.2 0 0>, rotation=<rot 90 0 0>}]
 dyn_leg_strap_l_01_MSTR [{shapeType=Capsule, length=0.505, radius=0.505}]
 dyn_leg_strap_l_02 [{shapeType=Capsule, length=5, radius=2, center=<3.2 0 0>, rotation=<rot 90 0 0>}]
 dyn_leg_strap_l_02_MSTR [{shapeType=Capsule, length=0.505, radius=0.505}]
 dyn_reg_races_r [{shapeType=Capsule, length=5, radius=2, center=<3.0121 0 0>, rotation=<rot 90 0 0>}]
 dyn_reg_races_r_MSTR [{shapeType=Capsule, length=0.505, radius=0.505}]
 dyn_reg_strap_r_01 [{shapeType=Capsule, length=5, radius=2, center=<3.2 0 0>, rotation=<rot 90 0 0>}]
 dyn_reg_strap_r_01_MSTR [{shapeType=Capsule, length=0.505, radius=0.505}]
 dyn_reg_strap_r_02 [{shapeType=Capsule, length=5, radius=2, center=<3.2 0 0>, rotation=<rot 90 0 0>}]
 dyn_reg_strap_r_02_MSTR [{shapeType=Capsule, length=0.505, radius=0.505}]
}
constraints [
 {
  bone1Name=dyn_collar_l_01
  bone2Name=dyn_collar_l_01_MSTR
  swing1LimitDegrees=0
  swing1Motion=Locked
  swing2LimitDegrees=20
  twistLimitDegrees=45
  twistMotion=Locked
 }
 {
  bone1Name=dyn_collar_r_01
  bone2Name=dyn_collar_r_01_MSTR
  swing1LimitDegrees=0
  swing1Motion=Locked
  swing2LimitDegrees=20
  twistLimitDegrees=45
  twistMotion=Locked
 }
 {bone1Name=dyn_leg_laces_l, bone2Name=dyn_leg_laces_l_MSTR, swing1LimitDegrees=20, swing2LimitDegrees=20, twistLimitDegrees=10}
 {
  bone1Name=dyn_leg_strap_l_01
  bone2Name=dyn_leg_strap_l_01_MSTR
  swing1LimitDegrees=5
  swing1Motion=Locked
  swing2LimitDegrees=45
  twistLimitDegrees=45
  twistMotion=Locked
 }
 {
  bone1Name=dyn_leg_strap_l_02
  bone2Name=dyn_leg_strap_l_02_MSTR
  swing1LimitDegrees=5
  swing1Motion=Locked
  swing2LimitDegrees=45
  twistLimitDegrees=45
  twistMotion=Locked
 }
 {bone1Name=dyn_reg_races_r, bone2Name=dyn_reg_races_r_MSTR, swing1LimitDegrees=20, swing2LimitDegrees=20, twistLimitDegrees=10}
 {
  bone1Name=dyn_reg_strap_r_01
  bone2Name=dyn_reg_strap_r_01_MSTR
  swing1LimitDegrees=5
  swing1Motion=Locked
  swing2LimitDegrees=45
  twistLimitDegrees=45
  twistMotion=Locked
 }
 {
  bone1Name=dyn_reg_strap_r_02
  bone2Name=dyn_reg_strap_r_02_MSTR
  swing1LimitDegrees=5
  swing1Motion=Locked
  swing2LimitDegrees=45
  twistLimitDegrees=45
  twistMotion=Locked
 }
 {
  bone1Name=dyn_butter_l_01
  bone2Name=dyn_butter_l_01_MSTR
  swing1LimitDegrees=45
  swing1Motion=Locked
  swing2LimitDegrees=20
  twistLimitDegrees=20
 }
 {
  bone1Name=dyn_butter_l_02
  bone2Name=dyn_butter_l_02_MSTR
  swing1LimitDegrees=45
  swing1Motion=Locked
  swing2LimitDegrees=20
  twistLimitDegrees=20
 }
 {
  bone1Name=dyn_butter_l_03
  bone2Name=dyn_butter_l_03_MSTR
  swing1LimitDegrees=45
  swing1Motion=Locked
  swing2LimitDegrees=20
  twistLimitDegrees=20
 }
 {
  bone1Name=dyn_butter_l_04
  bone2Name=dyn_butter_l_04_MSTR
  swing1LimitDegrees=45
  swing1Motion=Locked
  swing2LimitDegrees=20
  twistLimitDegrees=20
 }
 {
  bone1Name=dyn_butter_r_01
  bone2Name=dyn_butter_r_01_MSTR
  swing1LimitDegrees=45
  swing1Motion=Locked
  swing2LimitDegrees=20
  twistLimitDegrees=20
 }
 {
  bone1Name=dyn_butter_r_02
  bone2Name=dyn_butter_r_02_MSTR
  swing1LimitDegrees=45
  swing1Motion=Locked
  swing2LimitDegrees=20
  twistLimitDegrees=20
 }
 {
  bone1Name=dyn_butter_r_03
  bone2Name=dyn_butter_r_03_MSTR
  swing1LimitDegrees=45
  swing1Motion=Locked
  swing2LimitDegrees=20
  twistLimitDegrees=20
 }
 {
  bone1Name=dyn_butter_r_04
  bone2Name=dyn_butter_r_04_MSTR
  swing1LimitDegrees=45
  swing1Motion=Locked
  swing2LimitDegrees=20
  twistLimitDegrees=20
 }
]
