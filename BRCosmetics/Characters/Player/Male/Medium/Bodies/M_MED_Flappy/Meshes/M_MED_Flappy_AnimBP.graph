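# Animation Blueprint for the Flappy medium male body cosmetic, which follows the attached parent mesh's pose and adds secondary motion.
# A chain of weighted multi-bone Constraint nodes positions the "butter", leg strap and lace master bones relative to pelvis, spine, thigh and calf bones.
# A RigidBody node then simulates the Flappy Dyn physics asset in pelvis space, with overridden gravity and scaled component motion.
# The event graph holds no logic.
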AnimBlueprint /BRCosmetics/Characters/Player/Male/Medium/Bodies/M_MED_Flappy/Meshes/M_MED_Flappy_AnimBP
target_skeleton=&/M_MED_Flappy_Skeleton
graphs [
 AnimGraph {nodes [
   n1 = "Constraint - dyn_butter_l_01_MSTR (PPP)"!(
    ComponentPose=Local To Component(LocalPose=Copy Pose From Mesh() {bUseAttachedParent=true})
    Alpha=1
    ConstraintWeights_0=0.7249
    ConstraintWeights_1:float=0.1061
    ConstraintWeights_2:float=0.169
   ) -> (Pose) {
    alphaBoolBlend {blendOption=Linear}
    boneToModify {boneName=dyn_butter_l_01_MSTR}
    constraintSetup [{targetBone {boneName=pelvis}}, {targetBone {boneName=thigh_twist_01_l}}, {targetBone {boneName=spine_01}}]
    constraintWeights [1,1,1]
   }
   n2 = "Constraint - dyn_butter_l_02_MSTR (PPP)"!(ComponentPose=@n1, Alpha=1, ConstraintWeights_0=0.7706, ConstraintWeights_1:float=0.0959, ConstraintWeights_2:float=0.1335) -> (Pose) {
    alphaBoolBlend {blendOption=Linear}
    boneToModify {boneName=dyn_butter_l_02_MSTR}
    constraintSetup [{targetBone {boneName=pelvis}}, {targetBone {boneName=thigh_twist_01_l}}, {targetBone {boneName=spine_01}}]
    constraintWeights [1,1,1]
   }
   n3 = "Constraint - dyn_butter_l_03_MSTR (PPP)"!(ComponentPose=@n2, Alpha=1, ConstraintWeights_0=0.8282, ConstraintWeights_1:float=0.0671, ConstraintWeights_2:float=0.1047) -> (Pose) {
    alphaBoolBlend {blendOption=Linear}
    boneToModify {boneName=dyn_butter_l_03_MSTR}
    constraintSetup [{targetBone {boneName=pelvis}}, {targetBone {boneName=thigh_twist_01_l}}, {targetBone {boneName=spine_01}}]
    constraintWeights [1,1,1]
   }
   n4 = "Constraint - dyn_butter_l_04_MSTR (PPP)"!(ComponentPose=@n3, Alpha=1, ConstraintWeights_0=0.8704, ConstraintWeights_1:float=0.0345, ConstraintWeights_2:float=0.0951) -> (Pose) {
    alphaBoolBlend {blendOption=Linear}
    boneToModify {boneName=dyn_butter_l_04_MSTR}
    constraintSetup [{targetBone {boneName=pelvis}}, {targetBone {boneName=thigh_twist_01_l}}, {targetBone {boneName=spine_01}}]
    constraintWeights [1,1,1]
   }
   n5 = "Constraint - dyn_butter_r_04_MSTR (PPP)"!(ComponentPose=@n4, Alpha=1, ConstraintWeights_0=0.8703, ConstraintWeights_1:float=0.0345, ConstraintWeights_2:float=0.0951) -> (Pose) {
    alphaBoolBlend {blendOption=Linear}
    boneToModify {boneName=dyn_butter_r_04_MSTR}
    constraintSetup [{targetBone {boneName=pelvis}}, {targetBone {boneName=thigh_twist_01_r}}, {targetBone {boneName=spine_01}}]
    constraintWeights [1,1,1]
   }
   n6 = "Constraint - dyn_butter_r_03_MSTR (PPP)"!(ComponentPose=@n5, Alpha=1, ConstraintWeights_0=0.8261, ConstraintWeights_1:float=0.0682, ConstraintWeights_2:float=0.1057) -> (Pose) {
    alphaBoolBlend {blendOption=Linear}
    boneToModify {boneName=dyn_butter_r_03_MSTR}
    constraintSetup [{targetBone {boneName=pelvis}}, {targetBone {boneName=thigh_twist_01_r}}, {targetBone {boneName=spine_01}}]
    constraintWeights [1,1,1]
   }
   n7 = "Constraint - dyn_butter_r_02_MSTR (PPP)"!(ComponentPose=@n6, Alpha=1, ConstraintWeights_0=0.7706, ConstraintWeights_1:float=0.0959, ConstraintWeights_2:float=0.1335) -> (Pose) {
    alphaBoolBlend {blendOption=Linear}
    boneToModify {boneName=dyn_butter_r_02_MSTR}
    constraintSetup [{targetBone {boneName=pelvis}}, {targetBone {boneName=thigh_twist_01_r}}, {targetBone {boneName=spine_01}}]
    constraintWeights [1,1,1]
   }
   n8 = "Constraint - dyn_butter_r_01_MSTR (PPP)"!(ComponentPose=@n7, Alpha=1, ConstraintWeights_0=0.7249, ConstraintWeights_1:float=0.1061, ConstraintWeights_2:float=0.169) -> (Pose) {
    alphaBoolBlend {blendOption=Linear}
    boneToModify {boneName=dyn_butter_r_01_MSTR}
    constraintSetup [{targetBone {boneName=pelvis}}, {targetBone {boneName=thigh_twist_01_r}}, {targetBone {boneName=spine_01}}]
    constraintWeights [1,1,1]
   }
   n9 = "Constraint - dyn_leg_strap_l_01_MSTR (PPP)"!(ComponentPose=@n8, Alpha=1, ConstraintWeights_0=0.0324, ConstraintWeights_1:float=0.8326, ConstraintWeights_2:float=0.135) -> (Pose) {
    alphaBoolBlend {blendOption=Linear}
    boneToModify {boneName=dyn_leg_strap_l_01_MSTR}
    constraintSetup [{targetBone {boneName=pelvis}}, {targetBone {boneName=thigh_twist_01_l}}, {targetBone {boneName=thigh_l}}]
    constraintWeights [1,1,1]
   }
   n10 = "Constraint - dyn_leg_strap_l_02_MSTR (PP)"!(ComponentPose=@n9, Alpha=1, ConstraintWeights_0=0.4855, ConstraintWeights_1:float=0.5145) -> (Pose) {
    alphaBoolBlend {blendOption=Linear}
    boneToModify {boneName=dyn_leg_strap_l_02_MSTR}
    constraintSetup [{targetBone {boneName=thigh_twist_01_l}}, {targetBone {boneName=thigh_l}}]
    constraintWeights [1,1]
   }
   n11 = "Constraint - dyn_reg_strap_r_01_MSTR (PPP)"!(ComponentPose=@n10, Alpha=1, ConstraintWeights_0=0.1363, ConstraintWeights_1:float=0.0322, ConstraintWeights_2:float=0.8315) -> (Pose) {
    alphaBoolBlend {blendOption=Linear}
    boneToModify {boneName=dyn_reg_strap_r_01_MSTR}
    constraintSetup [{targetBone {boneName=thigh_r}}, {targetBone {boneName=pelvis}}, {targetBone {boneName=thigh_twist_01_r}}]
    constraintWeights [1,1,1]
   }
   n12 = "Constraint - dyn_reg_strap_r_02_MSTR (PP)"!(ComponentPose=@n11, Alpha=1, ConstraintWeights_0=0.5236, ConstraintWeights_1:float=0.4764) -> (Pose) {
    alphaBoolBlend {blendOption=Linear}
    boneToModify {boneName=dyn_reg_strap_r_02_MSTR}
    constraintSetup [{targetBone {boneName=thigh_r}}, {targetBone {boneName=thigh_twist_01_r}}]
    constraintWeights [1,1]
   }
   n13 = "Constraint - dyn_reg_races_r_MSTR (PPP)"!(ComponentPose=@n12, Alpha=1, ConstraintWeights_0=0.7247, ConstraintWeights_1:float=0.2113, ConstraintWeights_2:float=0.064) -> (Pose) {
    alphaBoolBlend {blendOption=Linear}
    boneToModify {boneName=dyn_reg_races_r_MSTR}
    constraintSetup [{targetBone {boneName=calf_twist_02_r}}, {targetBone {boneName=calf_twist_01_r}}, {targetBone {boneName=calf_r}}]
    constraintWeights [1,1,1]
   }
   n14 = "Constraint - dyn_leg_laces_l_MSTR (PPP)"!(ComponentPose=@n13, Alpha=1, ConstraintWeights_0=0.0509, ConstraintWeights_1:float=0.25, ConstraintWeights_2:float=0.699) -> (Pose) {
    alphaBoolBlend {blendOption=Linear}
    boneToModify {boneName=dyn_leg_laces_l_MSTR}
    constraintSetup [{targetBone {boneName=calf_l}}, {targetBone {boneName=calf_twist_01_l}}, {targetBone {boneName=calf_twist_02_l}}]
    constraintWeights [1,1,1]
   }
   n15 = RigidBody(ComponentPose=@n14) {
    bClampLinearTranslationLimitToRefPose=true
    bOverrideWorldGravity=true
    overridePhysicsAsset=&/M_MED_Flappy_Dyn
    simulationSpace=BaseBoneSpace
    alphaBoolBlend {blendOption=Linear}
    baseBoneRef {boneName=pelvis}
    componentAppliedLinearAccClamp=<3000 3000 3000>
    componentLinearAccScale=<0.4 0.3 0.3>
    componentLinearVelScale=<0.4 0.3 0.3>
    overrideWorldGravity {z=-980}
   }
   Output Pose(Result=Component To Local(ComponentPose=@n15)) {layerGroup=None}
  ]}
 EventGraph {nodes [Event BlueprintUpdateAnimation()]}
]
variables []
