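# Control Rig for the Flappy character's face accessory, carrying jaw and lip motion onto duplicate ALT face bones on a dynamic pancake-stack skeleton.
# The executed chain drives C_jaw from the Jaw control, then offset-preserving rotation and translation constraints move jaw and lower-mid-lip ALT bones, blending Secondary ones.
# An unwired chain copies local deltas from rest for jaw and all lip bones, at reduced weight on Secondary/Tertiary copies.
# Two duplicate lower-lip constraint chains are also unwired.
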ControlRigBlueprint /BRCosmetics/Characters/Player/Male/Medium/Bodies/M_MED_Flappy/Meshes/Parts/M_MED_Flappy_FaceAcc_ControlRig
graphs [
 RigVMFunctionLibrary {links [], nodes []}
 RigVMModel {nodes [
   RigUnit_BeginExecution = &RigUnit_BeginExecution()
   RigUnit_GetControlTransform = /Script/ControlRig.RigUnit_GetControlTransform(Control=Jaw, Space=GlobalSpace) {is=struct RigVMUnitNode, template="GetControlValue::Execute(in Control,in Space,out Transform,out Minimum,out Maximum)"}
   /Script/ControlRig.RigUnit_SetBoneTransform(
    Bone=C_jaw
    Space=GlobalSpace
    Weight=1
    bPropagateToChildren=True
    ExecutePin=@RigUnit_BeginExecution.ExecutePin
    Transform=@RigUnit_GetControlTransform.Transform
   ) {is=struct RigVMUnitNode}
   n1 = &RigUnit_GetBoneTransform(Bone=C_jaw)
   > /Script/ControlRig.RigUnit_TransformConstraint(
    BaseBone=None
    BaseTransform=(Rotation=(X=0,Y=0,Z=0,W=1),Translation=(X=0,Y=0,Z=0),Scale3D=(X=1,Y=1,Z=1))
    BaseTransformSpace=GlobalSpace
    Bone=C_jaw_ALT_Main
    Targets="((Transform=(Rotation=(X=0,Y=0,Z=0,W=1),Translation=(X=0,Y=0,Z=0),Scale3D=(X=1,Y=1,Z=1)),Weight=1,bMaintainOffset=True,Filter=(TranslationFilter=(bX=false,bY=false,bZ=false),RotationFilter=(bX=True,bY=True,bZ=True),ScaleFilter=(bX=false,bY=false,bZ=false)) ...(truncated)"
    bUseInitialTransforms=True
    Targets.0.Transform=@n1.Transform
   ) {is=struct RigVMUnitNode}
   > &RigUnit_TransformConstraint(
    BaseBone=None
    BaseTransform=(Rotation=(X=0,Y=0,Z=0,W=1),Translation=(X=0,Y=0,Z=0),Scale3D=(X=1,Y=1,Z=1))
    BaseTransformSpace=GlobalSpace
    Bone=C_jaw_ALT_Main
    Targets="((Transform=(Rotation=(X=0,Y=0,Z=0,W=1),Translation=(X=0,Y=0,Z=0),Scale3D=(X=1,Y=1,Z=1)),Weight=1,bMaintainOffset=True,Filter=(TranslationFilter=(bX=True,bY=True,bZ=True),RotationFilter=(bX=false,bY=false,bZ=false),ScaleFilter=(bX=false,bY=false,bZ=false)) ...(truncated)"
    bUseInitialTransforms=True
    Targets.0.Transform=@n1.Transform
   )
   n2 = &RigUnit_GetBoneTransform(Bone=faceAttach_lower_ALT_Secondary)
   > &RigUnit_TransformConstraint(
    BaseBone=None
    BaseTransform=(Rotation=(X=0,Y=0,Z=0,W=1),Translation=(X=0,Y=0,Z=0),Scale3D=(X=1,Y=1,Z=1))
    BaseTransformSpace=GlobalSpace
    Bone=C_jaw_ALT_Secondary
    Targets="((Transform=(Rotation=(X=0,Y=0,Z=0,W=1),Translation=(X=0,Y=0,Z=0),Scale3D=(X=1,Y=1,Z=1)),Weight=0.5,bMaintainOffset=True,Filter=(TranslationFilter=(bX=false,bY=false,bZ=false),RotationFilter=(bX=True,bY=True,bZ=True),ScaleFilter=(bX=false,bY=false,bZ=false ...(truncated)"
    bUseInitialTransforms=True
    Targets.0.Transform=@n1.Transform
    Targets.1.Transform=@n2.Transform
   )
   > &RigUnit_TransformConstraint(
    BaseBone=None
    BaseTransform=(Rotation=(X=0,Y=0,Z=0,W=1),Translation=(X=0,Y=0,Z=0),Scale3D=(X=1,Y=1,Z=1))
    BaseTransformSpace=GlobalSpace
    Bone=C_jaw_ALT_Secondary
    Targets="((Transform=(Rotation=(X=0,Y=0,Z=0,W=1),Translation=(X=0,Y=0,Z=0),Scale3D=(X=1,Y=1,Z=1)),Weight=0.5,bMaintainOffset=True,Filter=(TranslationFilter=(bX=True,bY=True,bZ=True),RotationFilter=(bX=false,bY=false,bZ=false),ScaleFilter=(bX=false,bY=false,bZ=false ...(truncated)"
    bUseInitialTransforms=True
    Targets.0.Transform=@n1.Transform
    Targets.1.Transform=@n2.Transform
   )
   n3 = &RigUnit_GetBoneTransform(Bone=C_lip_lower_mid)
   > &RigUnit_TransformConstraint(
    BaseBone=None
    BaseTransform=(Rotation=(X=0,Y=0,Z=0,W=1),Translation=(X=0,Y=0,Z=0),Scale3D=(X=1,Y=1,Z=1))
    BaseTransformSpace=GlobalSpace
    Bone=C_lip_lower_mid_ALT_Main
    Targets="((Transform=(Rotation=(X=0,Y=0,Z=0,W=1),Translation=(X=0,Y=0,Z=0),Scale3D=(X=1,Y=1,Z=1)),Weight=1,bMaintainOffset=True,Filter=(TranslationFilter=(bX=false,bY=false,bZ=false),RotationFilter=(bX=True,bY=True,bZ=True),ScaleFilter=(bX=false,bY=false,bZ=false)) ...(truncated)"
    bUseInitialTransforms=True
    Targets.0.Transform=@n3.Transform
   )
   > &RigUnit_TransformConstraint(
    BaseBone=None
    BaseTransform=(Rotation=(X=0,Y=0,Z=0,W=1),Translation=(X=0,Y=0,Z=0),Scale3D=(X=1,Y=1,Z=1))
    BaseTransformSpace=GlobalSpace
    Bone=C_lip_lower_mid_ALT_Main
    Targets="((Transform=(Rotation=(X=0,Y=0,Z=0,W=1),Translation=(X=0,Y=0,Z=0),Scale3D=(X=1,Y=1,Z=1)),Weight=1,bMaintainOffset=True,Filter=(TranslationFilter=(bX=True,bY=True,bZ=True),RotationFilter=(bX=false,bY=false,bZ=false),ScaleFilter=(bX=false,bY=false,bZ=false)) ...(truncated)"
    bUseInitialTransforms=True
    Targets.0.Transform=@n3.Transform
   )
   n4 = &RigUnit_GetBoneTransform(Bone=C_jaw_ALT_Secondary)
   > &RigUnit_TransformConstraint(
    BaseBone=None
    BaseTransform=(Rotation=(X=0,Y=0,Z=0,W=1),Translation=(X=0,Y=0,Z=0),Scale3D=(X=1,Y=1,Z=1))
    BaseTransformSpace=GlobalSpace
    Bone=C_lip_lower_mid_ALT_Secondary
    Targets="((Transform=(Rotation=(X=0,Y=0,Z=0,W=1),Translation=(X=0,Y=0,Z=0),Scale3D=(X=1,Y=1,Z=1)),Weight=0.5,bMaintainOffset=True,Filter=(TranslationFilter=(bX=false,bY=false,bZ=false),RotationFilter=(bX=True,bY=True,bZ=True),ScaleFilter=(bX=false,bY=false,bZ=false ...(truncated)"
    bUseInitialTransforms=True
    Targets.0.Transform=@n3.Transform
    Targets.1.Transform=@n4.Transform
   )
   > &RigUnit_TransformConstraint(
    BaseBone=None
    BaseTransform=(Rotation=(X=0,Y=0,Z=0,W=1),Translation=(X=0,Y=0,Z=0),Scale3D=(X=1,Y=1,Z=1))
    BaseTransformSpace=GlobalSpace
    Bone=C_lip_lower_mid_ALT_Secondary
    Targets="((Transform=(Rotation=(X=0,Y=0,Z=0,W=1),Translation=(X=0,Y=0,Z=0),Scale3D=(X=1,Y=1,Z=1)),Weight=0.5,bMaintainOffset=True,Filter=(TranslationFilter=(bX=True,bY=True,bZ=True),RotationFilter=(bX=false,bY=false,bZ=false),ScaleFilter=(bX=false,bY=false,bZ=false ...(truncated)"
    bUseInitialTransforms=True
    Targets.0.Transform=@n3.Transform
    Targets.1.Transform=@n4.Transform
   )
   RigUnit_GetInitialBoneTransform = /Script/ControlRig.RigUnit_GetInitialBoneTransform(Bone=C_jaw, Space=LocalSpace) {is=struct RigVMUnitNode}
   RigUnit_GetBoneTransform = &RigUnit_GetBoneTransform(Bone=C_jaw, Space=LocalSpace)
   n5 = &RigVMFunction_MathTransformMul(
    B=/Script/RigVM.RigVMFunction_MathTransformInverse(Value=@RigUnit_GetInitialBoneTransform.Transform) {is=struct RigVMUnitNode, template="Inverse::Execute(in Value,out Result)"}.Result
    A=@RigUnit_GetBoneTransform.Transform
   ) {template="Multiply::Execute(in A,in B,out Result)"}
   RigUnit_SetBoneTransform = &RigUnit_SetBoneTransform(
    Bone=C_jaw_ALT_Main
    ExecutePin=()
    Space=LocalSpace
    Weight=1
    bPropagateToChildren=True
    Transform=&RigVMFunction_MathTransformMul(A=@n5.Result, B=&RigUnit_GetBoneTransform(Bone=C_jaw_ALT_Main, Space=LocalSpace).Transform) {template="Multiply::Execute(in A,in B,out Result)"}.Result
   )
   > &RigUnit_SetBoneTransform(
    Bone=C_jaw_ALT_Secondary
    Space=LocalSpace
    Weight=0.5
    bPropagateToChildren=True
    Transform=&RigVMFunction_MathTransformMul(A=@n5.Result, B=&RigUnit_GetBoneTransform(Bone=C_jaw_ALT_Secondary, Space=LocalSpace).Transform) {template="Multiply::Execute(in A,in B,out Result)"}.Result
   )
   n6 = &RigVMFunction_MathTransformMul(
    B=&RigVMFunction_MathTransformInverse(Value=&RigUnit_GetInitialBoneTransform(Bone=C_lip_lower_mid, Space=LocalSpace).Transform) {template="Inverse::Execute(in Value,out Result)"}.Result
    A=&RigUnit_GetBoneTransform(Bone=C_lip_lower_mid, Space=LocalSpace).Transform
   ) {template="Multiply::Execute(in A,in B,out Result)"}
   > &RigUnit_SetBoneTransform(
    Bone=C_lip_lower_mid_ALT_Main
    Space=LocalSpace
    Weight=1
    bPropagateToChildren=True
    Transform=&RigVMFunction_MathTransformMul(A=@n6.Result, B=&RigUnit_GetBoneTransform(Bone=C_lip_lower_mid_ALT_Main, Space=LocalSpace).Transform) {template="Multiply::Execute(in A,in B,out Result)"}.Result
   )
   > &RigUnit_SetBoneTransform(
    Bone=C_lip_lower_mid_ALT_Secondary
    Space=LocalSpace
    Weight=0.5
    bPropagateToChildren=True
    Transform=&RigVMFunction_MathTransformMul(A=@n6.Result, B=&RigUnit_GetBoneTransform(Bone=C_lip_lower_mid_ALT_Secondary, Space=LocalSpace).Transform) {template="Multiply::Execute(in A,in B,out Result)"}.Result
   )
   n7 = &RigVMFunction_MathTransformMul(
    B=&RigVMFunction_MathTransformInverse(Value=&RigUnit_GetInitialBoneTransform(Bone=L_lip_lower_outer, Space=LocalSpace).Transform) {template="Inverse::Execute(in Value,out Result)"}.Result
    A=&RigUnit_GetBoneTransform(Bone=L_lip_lower_outer, Space=LocalSpace).Transform
   ) {template="Multiply::Execute(in A,in B,out Result)"}
   > &RigUnit_SetBoneTransform(
    Bone=L_lip_lower_outer_ALT_Main
    Space=LocalSpace
    Weight=1
    bPropagateToChildren=True
    Transform=&RigVMFunction_MathTransformMul(A=@n7.Result, B=&RigUnit_GetBoneTransform(Bone=L_lip_lower_outer_ALT_Main, Space=LocalSpace).Transform) {template="Multiply::Execute(in A,in B,out Result)"}.Result
   )
   > &RigUnit_SetBoneTransform(
    Bone=L_lip_lower_outer_ALT_Secondary
    Space=LocalSpace
    Weight=0.5
    bPropagateToChildren=True
    Transform=&RigVMFunction_MathTransformMul(A=@n7.Result, B=&RigUnit_GetBoneTransform(Bone=L_lip_lower_outer_ALT_Secondary, Space=LocalSpace).Transform) {template="Multiply::Execute(in A,in B,out Result)"}.Result
   )
   n8 = &RigVMFunction_MathTransformMul(
    B=&RigVMFunction_MathTransformInverse(Value=&RigUnit_GetInitialBoneTransform(Bone=R_lip_lower_outer, Space=LocalSpace).Transform) {template="Inverse::Execute(in Value,out Result)"}.Result
    A=&RigUnit_GetBoneTransform(Bone=R_lip_lower_outer, Space=LocalSpace).Transform
   ) {template="Multiply::Execute(in A,in B,out Result)"}
   > &RigUnit_SetBoneTransform(
    Bone=R_lip_lower_outer_ALT_Main
    Space=LocalSpace
    Weight=1
    bPropagateToChildren=True
    Transform=&RigVMFunction_MathTransformMul(A=@n8.Result, B=&RigUnit_GetBoneTransform(Bone=R_lip_lower_outer_ALT_Main, Space=LocalSpace).Transform) {template="Multiply::Execute(in A,in B,out Result)"}.Result
   )
   > &RigUnit_SetBoneTransform(
    Bone=R_lip_lower_outer_ALT_Secondary
    Space=LocalSpace
    Weight=0.5
    bPropagateToChildren=True
    Transform=&RigVMFunction_MathTransformMul(A=@n8.Result, B=&RigUnit_GetBoneTransform(Bone=R_lip_lower_outer_ALT_Secondary, Space=LocalSpace).Transform) {template="Multiply::Execute(in A,in B,out Result)"}.Result
   )
   n9 = &RigVMFunction_MathTransformMul(
    B=&RigVMFunction_MathTransformInverse(Value=&RigUnit_GetInitialBoneTransform(Bone=L_lip_upper_outer, Space=LocalSpace).Transform) {template="Inverse::Execute(in Value,out Result)"}.Result
    A=&RigUnit_GetBoneTransform(Bone=L_lip_upper_outer, Space=LocalSpace).Transform
   ) {template="Multiply::Execute(in A,in B,out Result)"}
   > &RigUnit_SetBoneTransform(
    Bone=L_lip_upper_outer_ALT_Main
    Space=LocalSpace
    Weight=1
    bPropagateToChildren=True
    Transform=&RigVMFunction_MathTransformMul(A=@n9.Result, B=&RigUnit_GetBoneTransform(Bone=L_lip_upper_outer_ALT_Main, Space=LocalSpace).Transform) {template="Multiply::Execute(in A,in B,out Result)"}.Result
   )
   > &RigUnit_SetBoneTransform(
    Bone=L_lip_upper_outer_ALT_Secondary
    Space=LocalSpace
    Weight=0.5
    bPropagateToChildren=True
    Transform=&RigVMFunction_MathTransformMul(A=@n9.Result, B=&RigUnit_GetBoneTransform(Bone=L_lip_upper_outer_ALT_Secondary, Space=LocalSpace).Transform) {template="Multiply::Execute(in A,in B,out Result)"}.Result
   )
   n10 = &RigVMFunction_MathTransformMul(
    B=&RigVMFunction_MathTransformInverse(Value=&RigUnit_GetInitialBoneTransform(Bone=R_lip_upper_outer, Space=LocalSpace).Transform) {template="Inverse::Execute(in Value,out Result)"}.Result
    A=&RigUnit_GetBoneTransform(Bone=R_lip_upper_outer, Space=LocalSpace).Transform
   ) {template="Multiply::Execute(in A,in B,out Result)"}
   > &RigUnit_SetBoneTransform(
    Bone=R_lip_upper_outer_ALT_Main
    Space=LocalSpace
    Weight=1
    bPropagateToChildren=True
    Transform=&RigVMFunction_MathTransformMul(A=@n10.Result, B=&RigUnit_GetBoneTransform(Bone=R_lip_upper_outer_ALT_Main, Space=LocalSpace).Transform) {template="Multiply::Execute(in A,in B,out Result)"}.Result
   )
   > &RigUnit_SetBoneTransform(
    Bone=R_lip_upper_outer_ALT_Secondary
    Space=LocalSpace
    Weight=0.5
    bPropagateToChildren=True
    Transform=&RigVMFunction_MathTransformMul(A=@n10.Result, B=&RigUnit_GetBoneTransform(Bone=R_lip_upper_outer_ALT_Secondary, Space=LocalSpace).Transform) {template="Multiply::Execute(in A,in B,out Result)"}.Result
   )
   n11 = &RigVMFunction_MathTransformMul(
    B=&RigVMFunction_MathTransformInverse(Value=&RigUnit_GetInitialBoneTransform(Bone=C_lip_upper_mid, Space=LocalSpace).Transform) {template="Inverse::Execute(in Value,out Result)"}.Result
    A=&RigUnit_GetBoneTransform(Bone=C_lip_upper_mid, Space=LocalSpace).Transform
   ) {template="Multiply::Execute(in A,in B,out Result)"}
   > &RigUnit_SetBoneTransform(
    Bone=C_lip_upper_mid_ALT_Main
    Space=LocalSpace
    Weight=1
    bPropagateToChildren=True
    Transform=&RigVMFunction_MathTransformMul(A=@n11.Result, B=&RigUnit_GetBoneTransform(Bone=C_lip_upper_mid_ALT_Main, Space=LocalSpace).Transform) {template="Multiply::Execute(in A,in B,out Result)"}.Result
   )
   > &RigUnit_SetBoneTransform(
    Bone=C_lip_upper_mid_ALT_Secondary
    Space=LocalSpace
    Weight=0.5
    bPropagateToChildren=True
    Transform=&RigVMFunction_MathTransformMul(A=@n11.Result, B=&RigUnit_GetBoneTransform(Bone=C_lip_upper_mid_ALT_Secondary, Space=LocalSpace).Transform) {template="Multiply::Execute(in A,in B,out Result)"}.Result
   )
   n12 = &RigVMFunction_MathTransformMul(
    B=&RigVMFunction_MathTransformInverse(Value=&RigUnit_GetInitialBoneTransform(Bone=L_lip_corner, Space=LocalSpace).Transform) {template="Inverse::Execute(in Value,out Result)"}.Result
    A=&RigUnit_GetBoneTransform(Bone=L_lip_corner, Space=LocalSpace).Transform
   ) {template="Multiply::Execute(in A,in B,out Result)"}
   > &RigUnit_SetBoneTransform(
    Bone=L_lip_corner_ALT_Main
    Space=LocalSpace
    Weight=1
    bPropagateToChildren=True
    Transform=&RigVMFunction_MathTransformMul(A=@n12.Result, B=&RigUnit_GetBoneTransform(Bone=L_lip_corner_ALT_Main, Space=LocalSpace).Transform) {template="Multiply::Execute(in A,in B,out Result)"}.Result
   )
   > &RigUnit_SetBoneTransform(
    Bone=L_lip_corner_ALT_Secondary
    Space=LocalSpace
    Weight=0.5
    bPropagateToChildren=True
    Transform=&RigVMFunction_MathTransformMul(A=@n12.Result, B=&RigUnit_GetBoneTransform(Bone=L_lip_corner_ALT_Secondary, Space=LocalSpace).Transform) {template="Multiply::Execute(in A,in B,out Result)"}.Result
   )
   > &RigUnit_SetBoneTransform(
    Bone=L_lip_corner_ALT_Tertiary
    Space=LocalSpace
    Weight=0.5
    bPropagateToChildren=True
    Transform=&RigVMFunction_MathTransformMul(A=@n12.Result, B=&RigUnit_GetBoneTransform(Bone=L_lip_corner_ALT_Tertiary, Space=LocalSpace).Transform) {template="Multiply::Execute(in A,in B,out Result)"}.Result
   )
   n13 = &RigVMFunction_MathTransformMul(
    B=&RigVMFunction_MathTransformInverse(Value=&RigUnit_GetInitialBoneTransform(Bone=R_lip_corner, Space=LocalSpace).Transform) {template="Inverse::Execute(in Value,out Result)"}.Result
    A=&RigUnit_GetBoneTransform(Bone=R_lip_corner, Space=LocalSpace).Transform
   ) {template="Multiply::Execute(in A,in B,out Result)"}
   > &RigUnit_SetBoneTransform(
    Bone=R_lip_corner_ALT_Main
    Space=LocalSpace
    Weight=1
    bPropagateToChildren=True
    Transform=&RigVMFunction_MathTransformMul(A=@n13.Result, B=&RigUnit_GetBoneTransform(Bone=R_lip_corner_ALT_Main, Space=LocalSpace).Transform) {template="Multiply::Execute(in A,in B,out Result)"}.Result
   )
   > &RigUnit_SetBoneTransform(
    Bone=R_lip_corner_ALT_Secondary
    Space=LocalSpace
    Weight=0.5
    bPropagateToChildren=True
    Transform=&RigVMFunction_MathTransformMul(A=@n13.Result, B=&RigUnit_GetBoneTransform(Bone=R_lip_corner_ALT_Secondary, Space=LocalSpace).Transform) {template="Multiply::Execute(in A,in B,out Result)"}.Result
   )
   > &RigUnit_SetBoneTransform(
    Bone=R_lip_corner_ALT_Tertiary
    Space=LocalSpace
    Weight=0.5
    bPropagateToChildren=True
    Transform=&RigVMFunction_MathTransformMul(A=@n13.Result, B=&RigUnit_GetBoneTransform(Bone=R_lip_corner_ALT_Tertiary, Space=LocalSpace).Transform) {template="Multiply::Execute(in A,in B,out Result)"}.Result
   )
   n14 = &RigUnit_GetBoneTransform(Bone=C_lip_lower_mid)
   &RigUnit_TransformConstraint(
    BaseBone=None
    BaseTransform=(Rotation=(X=0,Y=0,Z=0,W=1),Translation=(X=0,Y=0,Z=0),Scale3D=(X=1,Y=1,Z=1))
    BaseTransformSpace=GlobalSpace
    Bone=C_lip_lower_mid_ALT_Main
    ExecutePin=()
    Targets="((Transform=(Rotation=(X=0,Y=0,Z=0,W=1),Translation=(X=0,Y=0,Z=0),Scale3D=(X=1,Y=1,Z=1)),Weight=1,bMaintainOffset=True,Filter=(TranslationFilter=(bX=false,bY=false,bZ=false),RotationFilter=(bX=True,bY=True,bZ=True),ScaleFilter=(bX=false,bY=false,bZ=false)) ...(truncated)"
    bUseInitialTransforms=True
    Targets.0.Transform=@n14.Transform
   )
   > &RigUnit_TransformConstraint(
    BaseBone=None
    BaseTransform=(Rotation=(X=0,Y=0,Z=0,W=1),Translation=(X=0,Y=0,Z=0),Scale3D=(X=1,Y=1,Z=1))
    BaseTransformSpace=GlobalSpace
    Bone=C_lip_lower_mid_ALT_Main
    Targets="((Transform=(Rotation=(X=0,Y=0,Z=0,W=1),Translation=(X=0,Y=0,Z=0),Scale3D=(X=1,Y=1,Z=1)),Weight=1,bMaintainOffset=True,Filter=(TranslationFilter=(bX=True,bY=True,bZ=True),RotationFilter=(bX=false,bY=false,bZ=false),ScaleFilter=(bX=false,bY=false,bZ=false)) ...(truncated)"
    bUseInitialTransforms=True
    Targets.0.Transform=@n14.Transform
   )
   n15 = &RigUnit_GetBoneTransform(Bone=C_jaw_ALT_Secondary)
   > &RigUnit_TransformConstraint(
    BaseBone=None
    BaseTransform=(Rotation=(X=0,Y=0,Z=0,W=1),Translation=(X=0,Y=0,Z=0),Scale3D=(X=1,Y=1,Z=1))
    BaseTransformSpace=GlobalSpace
    Bone=C_lip_lower_mid_ALT_Secondary
    Targets="((Transform=(Rotation=(X=0,Y=0,Z=0,W=1),Translation=(X=0,Y=0,Z=0),Scale3D=(X=1,Y=1,Z=1)),Weight=0.5,bMaintainOffset=True,Filter=(TranslationFilter=(bX=false,bY=false,bZ=false),RotationFilter=(bX=True,bY=True,bZ=True),ScaleFilter=(bX=false,bY=false,bZ=false ...(truncated)"
    bUseInitialTransforms=True
    Targets.0.Transform=@n14.Transform
    Targets.1.Transform=@n15.Transform
   )
   > &RigUnit_TransformConstraint(
    BaseBone=None
    BaseTransform=(Rotation=(X=0,Y=0,Z=0,W=1),Translation=(X=0,Y=0,Z=0),Scale3D=(X=1,Y=1,Z=1))
    BaseTransformSpace=GlobalSpace
    Bone=C_lip_lower_mid_ALT_Secondary
    Targets="((Transform=(Rotation=(X=0,Y=0,Z=0,W=1),Translation=(X=0,Y=0,Z=0),Scale3D=(X=1,Y=1,Z=1)),Weight=0.5,bMaintainOffset=True,Filter=(TranslationFilter=(bX=True,bY=True,bZ=True),RotationFilter=(bX=false,bY=false,bZ=false),ScaleFilter=(bX=false,bY=false,bZ=false ...(truncated)"
    bUseInitialTransforms=True
    Targets.0.Transform=@n14.Transform
    Targets.1.Transform=@n15.Transform
   )
   n16 = &RigUnit_GetBoneTransform(Bone=C_lip_lower_mid)
   &RigUnit_TransformConstraint(
    BaseBone=None
    BaseTransform=(Rotation=(X=0,Y=0,Z=0,W=1),Translation=(X=0,Y=0,Z=0),Scale3D=(X=1,Y=1,Z=1))
    BaseTransformSpace=GlobalSpace
    Bone=C_lip_lower_mid_ALT_Main
    ExecutePin=()
    Targets="((Transform=(Rotation=(X=0,Y=0,Z=0,W=1),Translation=(X=0,Y=0,Z=0),Scale3D=(X=1,Y=1,Z=1)),Weight=1,bMaintainOffset=True,Filter=(TranslationFilter=(bX=false,bY=false,bZ=false),RotationFilter=(bX=True,bY=True,bZ=True),ScaleFilter=(bX=false,bY=false,bZ=false)) ...(truncated)"
    bUseInitialTransforms=True
    Targets.0.Transform=@n16.Transform
   )
   > &RigUnit_TransformConstraint(
    BaseBone=None
    BaseTransform=(Rotation=(X=0,Y=0,Z=0,W=1),Translation=(X=0,Y=0,Z=0),Scale3D=(X=1,Y=1,Z=1))
    BaseTransformSpace=GlobalSpace
    Bone=C_lip_lower_mid_ALT_Main
    Targets="((Transform=(Rotation=(X=0,Y=0,Z=0,W=1),Translation=(X=0,Y=0,Z=0),Scale3D=(X=1,Y=1,Z=1)),Weight=1,bMaintainOffset=True,Filter=(TranslationFilter=(bX=True,bY=True,bZ=True),RotationFilter=(bX=false,bY=false,bZ=false),ScaleFilter=(bX=false,bY=false,bZ=false)) ...(truncated)"
    bUseInitialTransforms=True
    Targets.0.Transform=@n16.Transform
   )
   n17 = &RigUnit_GetBoneTransform(Bone=C_jaw_ALT_Secondary)
   > &RigUnit_TransformConstraint(
    BaseBone=None
    BaseTransform=(Rotation=(X=0,Y=0,Z=0,W=1),Translation=(X=0,Y=0,Z=0),Scale3D=(X=1,Y=1,Z=1))
    BaseTransformSpace=GlobalSpace
    Bone=C_lip_lower_mid_ALT_Secondary
    Targets="((Transform=(Rotation=(X=0,Y=0,Z=0,W=1),Translation=(X=0,Y=0,Z=0),Scale3D=(X=1,Y=1,Z=1)),Weight=0.5,bMaintainOffset=True,Filter=(TranslationFilter=(bX=false,bY=false,bZ=false),RotationFilter=(bX=True,bY=True,bZ=True),ScaleFilter=(bX=false,bY=false,bZ=false ...(truncated)"
    bUseInitialTransforms=True
    Targets.0.Transform=@n16.Transform
    Targets.1.Transform=@n17.Transform
   )
   > &RigUnit_TransformConstraint(
    BaseBone=None
    BaseTransform=(Rotation=(X=0,Y=0,Z=0,W=1),Translation=(X=0,Y=0,Z=0),Scale3D=(X=1,Y=1,Z=1))
    BaseTransformSpace=GlobalSpace
    Bone=C_lip_lower_mid_ALT_Secondary
    Targets="((Transform=(Rotation=(X=0,Y=0,Z=0,W=1),Translation=(X=0,Y=0,Z=0),Scale3D=(X=1,Y=1,Z=1)),Weight=0.5,bMaintainOffset=True,Filter=(TranslationFilter=(bX=True,bY=True,bZ=True),RotationFilter=(bX=false,bY=false,bZ=false),ScaleFilter=(bX=false,bY=false,bZ=false ...(truncated)"
    bUseInitialTransforms=True
    Targets.0.Transform=@n16.Transform
    Targets.1.Transform=@n17.Transform
   )
  ]}
]
hierarchy [
 root {loc=<0 0 0>, rot=<0 0 0 1>}
 pelvis {parent=root, loc=<0 -0.7071 93.1191>, rot=<0.0127 0.707 -0.0127 -0.707>}
 spine_01 {parent=pelvis, loc=<6.7494 -0.6388 0>, rot=<0 0 0.034 -0.9994>}
 spine_02 {parent=spine_01, loc=<7.2518 -0.035 0>, rot=<0 0 0.0238 -0.9997>}
 spine_03 {parent=spine_02, loc=<6.0104 0 0>, rot=<0 0 0.0149 -0.9999>}
 spine_04 {parent=spine_03, loc=<8.1024 0 0>, rot=<0 0 -0.0132 -0.9999>}
 spine_05 {parent=spine_04, loc=<10.0146 0 0>, rot=<0 0 -0.0015 -1>}
 neck_01 {parent=spine_05, loc=<11.3138 0.1529 0>, rot=<0 0 -0.2003 -0.9797>}
 neck_02 {parent=neck_01, loc=<4.5627 0 0>, rot=<0 0 0.0504 -0.9987>}
 head {parent=neck_02, loc=<5.3914 0 0>, rot=<0 0 0.115 -0.9934>}
 faceAttach {parent=head, loc=<-0.8266 -0.1832 0>, rot=<0.0044 -0.3827 -0.0044 -0.9239>, scale=<0.89 0.89 0.89>}
 C_brow_mid {parent=faceAttach, loc=<4.7209 11.2084 4.8303>, rot=<-0.0586 -0.317 0.0114 -0.9466>}
 L_brow_mid {parent=faceAttach, loc=<7.639 10.371 3.157>, rot=<0.0603 -0.3852 0.2489 -0.8866>}
 L_brow_outer {parent=faceAttach, loc=<9.1042 8.1218 1.4857>, rot=<0.2349 -0.2525 0.3549 -0.869>}
 L_cheek_inner {parent=faceAttach, loc=<4.2785 8.761 -1.4435>, rot=<0.0454 -0.6262 0.3191 -0.7099>}
 C_lip_upper_mid {parent=faceAttach, loc=<-0.3154 10.7989 -0.2464>, rot=<0.2097 -0.3713 -0.0897 -0.9001>}
 L_lip_upper_outer {parent=faceAttach, loc=<0.8467 10.1076 -1.4454>, rot=<0.1857 -0.3758 0.3015 -0.8564>}
 L_lip_corner {parent=faceAttach, loc=<1.613 9.1457 -2.3124>, rot=<0.0012 -0.375 -0.0189 -0.9269>}
 C_nose_bridge {parent=faceAttach, loc=<2.3602 13.1515 2.4853>, rot=<-0.2002 -0.3676 0.0825 -0.9044>}
 R_brow_mid {parent=faceAttach, loc=<3.0903 10.342 7.7132>, rot=<-0.1347 -0.3556 -0.2231 -0.8976>}
 R_brow_outer {parent=faceAttach, loc=<1.4121 8.0675 9.1444>, rot=<-0.1251 -0.3913 -0.4176 -0.8104>}
 R_cheek_inner {parent=faceAttach, loc=<-1.4992 8.7305 4.3237>, rot=<-0.1979 -0.0545 -0.2598 -0.9436>}
 R_lip_corner {parent=faceAttach, loc=<-2.32 9.1134 1.692>, rot=<0.0127 -0.39 0.008 -0.9207>}
 C_jaw {parent=faceAttach, loc=<0.3123 3.2483 0.3583>, rot=<0.0044 -0.3827 -0.0044 -0.9239>}
 L_lip_lower_outer {parent=C_jaw, loc=<1.6966 6.8705 -0.8455>, rot=<-0.0146 -0.0279 0.3432 -0.9387>}
 C_lip_lower_mid {parent=C_jaw, loc=<0.0184 7.4191 -0.884>, rot=<0 -0.0099 -0.0021 -0.9999>}
 R_lip_lower_outer {parent=C_jaw, loc=<-1.6982 6.9042 -0.8853>, rot=<0.0021 -0.0171 -0.3152 -0.9489>}
 teeth_lower {parent=C_jaw, loc=<0.0169 5.3598 -2.8149>, rot=<0 0 0 1>}
 tongue {parent=C_jaw, loc=<0.0178 1.5245 -1.0398>, rot=<-0.1568 0 0 -0.9876>}
 R_eye {parent=faceAttach, loc=<2.2537 8.4351 6.6885>, rot=<0.3823 -0.0165 0.9223 -0.0548>}
 R_eye_lid_upper_mid {parent=R_eye, loc=<-0.0465 -1.169 0.2722>, rot=<0.0559 -0.0011 0.0197 -0.9982>, scale=<1 1.0001 0.9999>}
 R_eye_lid_lower_mid {parent=R_eye, loc=<-0.0141 -0.9451 -0.7006>, rot=<0.035 -0.0006 0.0197 -0.9992>}
 L_eye {parent=faceAttach, loc=<6.6314 8.4647 2.3108>, rot=<0.3819 0.0253 0.9227 0.046>}
 L_eye_lid_upper_mid {parent=L_eye, loc=<0.0251 -1.1623 0.2388>, rot=<0.0523 0 0 -0.9986>}
 L_eye_lid_lower_mid {parent=L_eye, loc=<0.0223 -0.923 -0.7488>, rot=<0.0329 0 0 -0.9995>}
 teeth_upper {parent=faceAttach, loc=<0.0954 9.3381 0.1585>, rot=<0.0044 -0.3827 -0.0044 -0.9239>}
 R_lip_upper_outer {parent=faceAttach, loc=<-1.497 10.1359 0.962>, rot=<-0.0848 -0.2926 -0.3215 -0.8966>}
 neck_01_ALT_A {parent=spine_05, loc=<10.9294 4.2357 0>, rot=<0 0 -0.2003 -0.9797>}
 neck_01_ALT_B {parent=neck_01_ALT_A, loc=<2.3802 -1.4414 0>, rot=<0 0 0 1>}
 neck_02_ALT_C {parent=neck_01_ALT_B, loc=<3.03 0.2732 0>, rot=<0 0 0.0504 -0.9987>}
 neck_02_ALT_D {parent=neck_02_ALT_C, loc=<2.6172 -1.0493 0>, rot=<0 0 0 -1>}
 head_ALT_E {parent=neck_02_ALT_D, loc=<3.0025 0.8306 0>, rot=<0 0 0.115 -0.9934>}
 head_ALT_F {parent=head_ALT_E, loc=<3.403 -1.2611 0>, rot=<0 0 0 -1>}
 head_ALT_G {parent=head_ALT_F, loc=<3.2129 -1.2557 0>, rot=<0 0 0 -1>}
 dyn_pancake_base_G {parent=head_ALT_G, loc=<0.0643 -0.111 -0.4893>, rot=<0.9564 0.0601 0.0013 0.2859>}
 dyn_pancake_G_side_01 {parent=dyn_pancake_base_G, loc=<0 0 0.7281>, rot=<0 0.7071 0 -0.7071>}
 dyn_syrup_A_01 {parent=dyn_pancake_G_side_01, loc=<10.0446 -1.1338 0.9954>, rot=<0.7071 0 0.7071 0>}
 dyn_syrup_J_01 {parent=dyn_pancake_G_side_01, loc=<8.33 -7.2189 0.4427>, rot=<0.6929 -0.141 0.6929 0.141>}
 dyn_syrup_J_02 {parent=dyn_syrup_J_01, loc=<4.0916 0 0>, rot=<0 0 0 -1>}
 dyn_syrup_B_01 {parent=dyn_pancake_G_side_01, loc=<7.6091 5.2495 0.222>, rot=<0.6688 0.2296 0.6688 -0.2296>}
 dyn_syrup_B_02 {parent=dyn_syrup_B_01, loc=<5.9012 0 0>, rot=<0 0 0 -1>}
 dyn_pancake_G_side_02 {parent=dyn_pancake_base_G, loc=<0 0 -0.8571>, rot=<0.7071 0 -0.7071 0>}
 dyn_syrup_D_01 {parent=dyn_pancake_G_side_02, loc=<5.1786 -9.399 2.0037>, rot=<0.5909 -0.3884 0.5909 0.3884>}
 dyn_syrup_D_02 {parent=dyn_syrup_D_01, loc=<5.0939 0 0>, rot=<0 0 0 -1>}
 dyn_syrup_E_01 {parent=dyn_pancake_G_side_02, loc=<9.9613 -5.6542 2.5616>, rot=<0.6874 -0.1657 0.6737 0.2149>}
 dyn_syrup_E_02 {parent=dyn_syrup_E_01, loc=<5.0944 0 0>, rot=<0 0 0 1>}
 dyn_syrup_F_01 {parent=dyn_pancake_G_side_02, loc=<12.101 3.6349 1.0046>, rot=<0.7018 0.0868 0.7018 -0.0868>}
 dyn_syrup_F_02 {parent=dyn_syrup_F_01, loc=<5.3277 0 0>, rot=<0 0 0 -1>}
 dyn_syrup_F_03 {parent=dyn_syrup_F_02, loc=<5.4379 0 0>, rot=<0 0 0 -1>}
 dyn_syrup_G_01 {parent=dyn_pancake_G_side_02, loc=<7.8185 9.1967 0.2719>, rot=<0.6521 0.3501 0.5925 -0.3181>}
 dyn_syrup_G_02 {parent=dyn_syrup_G_01, loc=<6.6103 0 0>, rot=<0 0 0 1>}
 dyn_syrup_G_03 {parent=dyn_syrup_G_02, loc=<5.7056 0 0>, rot=<0 0 0 1>}
 dyn_syrup_G_04 {parent=dyn_syrup_G_03, loc=<6.1202 0 0>, rot=<0 0 0 1>}
 dyn_syrup_H_01 {parent=dyn_pancake_G_side_02, loc=<1.901 12.66 0.3805>, rot=<-0.5177 -0.5409 -0.4583 0.4788>}
 dyn_syrup_H_02 {parent=dyn_syrup_H_01, loc=<4.4406 0 0>, rot=<0 0 0 -1>}
 dyn_syrup_H_03 {parent=dyn_syrup_H_02, loc=<3.8932 0 0>, rot=<0 0 0 -1>}
 dyn_syrup_C_01 {parent=dyn_pancake_base_G, loc=<-0.9409 9.8758 2.6245>, rot=<0 0.6349 0.7726 0>}
 dyn_syrup_C_02 {parent=dyn_syrup_C_01, loc=<4.4075 0 0>, rot=<0 0 0 -1>}
 dyn_syrup_I_01 {parent=dyn_pancake_base_G, loc=<0.0231 -11.9468 4.5786>, rot=<0.0071 -0.5171 0.855 -0.0383>}
 dyn_syrup_I_02 {parent=dyn_syrup_I_01, loc=<4.9648 0 0>, rot=<0 0 0 -1>}
 dyn_syrup_I_03 {parent=dyn_syrup_I_02, loc=<5.2379 0 0>, rot=<0 0 0 -1>}
 dyn_butter {parent=dyn_pancake_base_G, loc=<3.6587 -1.7417 -1.6259>, rot=<0 0 0 -1>}
 dyn_pancake_base_F {parent=head_ALT_F, loc=<-0.0317 0.0533 -0.4744>, rot=<-0.2476 -0.014 -0.0755 -0.9658>}
 dyn_pancake_F_side_01 {parent=dyn_pancake_base_F, loc=<0 0 0.7281>, rot=<0 0.7071 0 -0.7071>}
 dyn_pancake_F_side_02 {parent=dyn_pancake_base_F, loc=<0 0 -0.8571>, rot=<0.7071 0 -0.7071 0>}
 dyn_pancake_base_E {parent=head_ALT_E, loc=<0.0945 0.0811 0.1443>, rot=<0.4769 0.0251 -0.0505 -0.8772>}
 dyn_pancake_E_side_01 {parent=dyn_pancake_base_E, loc=<0 0 0.7281>, rot=<0 0.7071 0 -0.7071>}
 dyn_pancake_E_side_02 {parent=dyn_pancake_base_E, loc=<0 0 -0.8571>, rot=<0.7071 0 -0.7071 0>}
 faceAttach_upper_ALT_Secondary {parent=dyn_pancake_base_E, loc=<-2.6495 -1.5534 -2.6825>, rot=<-0.4562 -0.3569 -0.1398 -0.8031>, scale=<0.89 0.89 0.89>}
 C_lip_upper_mid_ALT_Secondary {parent=faceAttach_upper_ALT_Secondary, loc=<1.2041 13.6883 -0.4436>, rot=<0.2097 -0.3713 -0.0897 -0.9001>}
 L_lip_upper_outer_ALT_Secondary {parent=faceAttach_upper_ALT_Secondary, loc=<4.7269 12.6257 -3.869>, rot=<0.1857 -0.3758 0.3015 -0.8564>}
 L_lip_corner_ALT_Secondary {parent=faceAttach_upper_ALT_Secondary, loc=<6.8492 10.4321 -5.6349>, rot=<0.1277 -0.3526 0.295 -0.8789>, scale=<1.1236 1.1236 1.1236>}
 R_lip_corner_ALT_Secondary {parent=faceAttach_upper_ALT_Secondary, loc=<-3.9232 10.4001 5.2619>, rot=<-0.0911 -0.3794 -0.2361 -0.8899>, scale=<1.1236 1.1236 1.1236>}
 R_lip_upper_outer_ALT_Secondary {parent=faceAttach_upper_ALT_Secondary, loc=<-2.2156 12.6474 3.1829>, rot=<-0.0848 -0.2926 -0.3215 -0.8966>}
 dyn_pancake_base_D {parent=neck_02_ALT_D, loc=<0.0312 -0.0499 -1.1561>, rot=<-0.7119 0.0322 0.0198 -0.7013>}
 dyn_pancake_D_side_01 {parent=dyn_pancake_base_D, loc=<0 0 0.7281>, rot=<0 0.7071 0 -0.7071>}
 faceAttach_upper_ALT_Main {parent=dyn_pancake_base_D, loc=<0.2402 1.1804 1.7961>, rot=<0.6367 -0.2248 0.3255 -0.6619>, scale=<0.89 0.89 0.89>}
 C_lip_upper_mid_ALT_Main {parent=faceAttach_upper_ALT_Main, loc=<-1.7709 13.702 -3.4185>, rot=<0.2097 -0.3713 -0.0897 -0.9001>}
 L_lip_upper_outer_ALT_Main {parent=faceAttach_upper_ALT_Main, loc=<1.7519 12.6395 -6.8439>, rot=<0.1857 -0.3758 0.3015 -0.8564>}
 L_lip_corner_ALT_Main {parent=faceAttach_upper_ALT_Main, loc=<3.8742 10.4459 -8.6098>, rot=<0.1277 -0.3526 0.295 -0.8789>, scale=<1.1236 1.1236 1.1236>}
 R_lip_corner_ALT_Main {parent=faceAttach_upper_ALT_Main, loc=<-6.8982 10.4138 2.2869>, rot=<-0.0911 -0.3794 -0.2361 -0.8899>, scale=<1.1236 1.1236 1.1236>}
 R_lip_upper_outer_ALT_Main {parent=faceAttach_upper_ALT_Main, loc=<-5.1906 12.6611 0.208>, rot=<-0.0848 -0.2926 -0.3215 -0.8966>}
 dyn_pancake_base_C {parent=neck_02_ALT_C, loc=<-0.0229 -0.1209 -1.0537>, rot=<0.7029 -0.0211 0.0138 -0.7108>}
 dyn_pancake_C_side_02 {parent=dyn_pancake_base_C, loc=<0 0 -0.8571>, rot=<0.7071 0 -0.7071 0>}
 faceAttach_lower_ALT_Main {parent=dyn_pancake_base_C, loc=<2.8817 -1.1135 -2.6522>, rot=<-0.6705 -0.3233 -0.2087 -0.6343>, scale=<0.89 0.89 0.89>}
 C_jaw_ALT_Main {parent=faceAttach_lower_ALT_Main, loc=<-0.6314 1.9789 -2.2759>, rot=<0.0044 -0.3827 -0.0044 -0.9239>}
 L_lip_lower_outer_ALT_Main {parent=C_jaw_ALT_Main, loc=<4.9976 10.6684 -1.4685>, rot=<-0.0146 -0.0279 0.3432 -0.9387>}
 C_lip_lower_mid_ALT_Main {parent=C_jaw_ALT_Main, loc=<0.0542 11.522 -1.6494>, rot=<0 -0.0099 -0.0021 -0.9999>}
 R_lip_lower_outer_ALT_Main {parent=C_jaw_ALT_Main, loc=<-5.0021 10.7209 -1.5369>, rot=<0.0021 -0.0171 -0.3152 -0.9489>}
 dyn_pancake_base_B {parent=neck_01_ALT_B, loc=<0.1426 -0.155 -0.8801>, rot=<0.6424 -0.0441 0.0251 -0.7647>}
 dyn_pancake_B_side_01 {parent=dyn_pancake_base_B, loc=<0 0 0.7281>, rot=<0 0.7071 0 -0.7071>}
 dyn_pancake_B_side_02 {parent=dyn_pancake_base_B, loc=<0 0 -0.8571>, rot=<0.7071 0 -0.7071 0>}
 faceAttach_lower_ALT_Secondary {parent=dyn_pancake_base_B, loc=<5.5178 -1.4418 -2.0227>, rot=<-0.6278 -0.3455 -0.1552 -0.68>, scale=<0.89 0.89 0.89>}
 C_jaw_ALT_Secondary {parent=faceAttach_lower_ALT_Secondary, loc=<-3.7954 1.9935 -5.4398>, rot=<0.0044 -0.3827 -0.0044 -0.9239>}
 L_lip_lower_outer_ALT_Secondary {parent=C_jaw_ALT_Secondary, loc=<4.9976 10.6684 -1.4685>, rot=<-0.0146 -0.0279 0.3432 -0.9387>}
 C_lip_lower_mid_ALT_Secondary {parent=C_jaw_ALT_Secondary, loc=<0.0542 11.522 -1.6494>, rot=<0 -0.0099 -0.0021 -0.9999>}
 R_lip_lower_outer_ALT_Secondary {parent=C_jaw_ALT_Secondary, loc=<-5.0021 10.7209 -1.5369>, rot=<0.0021 -0.0171 -0.3152 -0.9489>}
 L_lip_corner_ALT_Tertiary {parent=faceAttach_lower_ALT_Secondary, loc=<1.4051 10.3876 -11.1635>, rot=<0.1277 -0.3526 0.295 -0.8789>, scale=<1.1236 1.1236 1.1236>}
 R_lip_corner_ALT_Tertiary {parent=faceAttach_lower_ALT_Secondary, loc=<-9.4503 10.3532 -0.1837>, rot=<-0.0911 -0.3794 -0.2361 -0.8899>, scale=<1.1236 1.1236 1.1236>}
 dyn_pancake_base_A {parent=neck_01_ALT_A, loc=<0.1702 0.3324 -0.6781>, rot=<0 0 0.0383 -0.9993>}
 dyn_pancake_A_side_01 {parent=dyn_pancake_base_A, loc=<0 0 0.7281>, rot=<0 0.7071 0 -0.7071>}
 dyn_pancake_A_side_02 {parent=dyn_pancake_base_A, loc=<0 0 -0.8571>, rot=<0.7071 0 -0.7071 0>}
 Jaw {type=Control, loc=<0 0 0>, rot=<0 0 0 1>}
]
