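# Physics asset for the Flappy male cosmetic's face accessory, giving dynamic bones to a pancake stack with butter and syrup drips.
# Pancake bases B–G lock rotationally to neck and head, while base A rotates freely on spine_05.
# Side flaps get narrow twist-locked swing; butter and multi-bone syrup chains hang from the G pancake with limited rotation.
# Large capsules cover head, neck, upper spine and right upper arm.
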PhysicsAsset /BRCosmetics/Characters/Player/Male/Medium/Bodies/M_MED_Flappy/Meshes/Parts/M_MED_Flappy_FaceAcc_Dyn
bodies {
 dyn_butter [{shapeType=Capsule, length=2, radius=1, center=<0.5706 -0.0002 0>}]
 dyn_pancake_B_side_02 [{shapeType=Capsule, length=2, radius=1, center=<0.9029 0.0002 0>, rotation=<rot 90 0 0>}]
 dyn_pancake_C_side_02 [{shapeType=Capsule, length=2, radius=1, center=<0.9029 0.0002 0>, rotation=<rot 90 0 0>}]
 dyn_pancake_D_side_01 [{shapeType=Capsule, length=2, radius=1, center=<0.9029 0.0002 0>, rotation=<rot 90 0 0>}]
 dyn_pancake_E_side_02 [{shapeType=Capsule, length=2, radius=1, center=<0.9029 0.0002 0>, rotation=<rot 90 0 0>}]
 dyn_pancake_F_side_01 [{shapeType=Capsule, length=2, radius=1, center=<0.9029 0.0002 0>, rotation=<rot 90 0 0>}]
 dyn_pancake_F_side_02 [{shapeType=Capsule, length=2, radius=1, center=<0.9029 0.0002 0>, rotation=<rot 90 0 0>}]
 dyn_pancake_G_side_01 [{shapeType=Capsule, length=2, radius=1, center=<0.9029 0.0002 0>, rotation=<rot 90 0 0>}]
 dyn_pancake_G_side_02 [{shapeType=Capsule, length=2, radius=1, center=<0.9029 0.0002 0>, rotation=<rot 90 0 0>}]
 dyn_pancake_base_A [{shapeType=Capsule, length=1, radius=1, shapeName=dyn_pancake_base_A_capsule, center=<0.081 -0.3693 0.6257>, rotation=<rot 8.4888 -90 0.0973>}]
 dyn_pancake_base_B [{shapeType=Capsule, length=2, radius=1}]
 dyn_pancake_base_C [{shapeType=Capsule, length=2, radius=1}]
 dyn_pancake_base_D [{shapeType=Capsule, length=2, radius=1}]
 dyn_pancake_base_E [{shapeType=Capsule, length=2, radius=1}]
 dyn_pancake_base_F [{shapeType=Capsule, length=2, radius=1}]
 dyn_pancake_base_G [{shapeType=Capsule, length=2, radius=1}]
 dyn_syrup_A_01 [{shapeType=Capsule, length=2, radius=1, center=<1.8 0 0>, rotation=<rot 90 0 0>}]
 dyn_syrup_B_01 [{shapeType=Capsule, length=2.9688, radius=1.5, center=<2.9585 -0.0001 0>, rotation=<rot 90 0 0>}]
 dyn_syrup_B_02 [{shapeType=Capsule, length=0.75, radius=2.25, center=<1.8 0 0>, rotation=<rot 90 0 0>}]
 dyn_syrup_C_01 [{shapeType=Capsule, length=2.1562, radius=1.5, center=<1.1422 0 0>, rotation=<rot 90 0 0>}]
 dyn_syrup_C_02 [{shapeType=Capsule, length=1.1562, radius=2.25, center=<2.0906 0 0>, rotation=<rot 90 0 0>}]
 dyn_syrup_D_01 [{shapeType=Capsule, length=2.3125, radius=2, center=<2 -1 0>, rotation=<rot 90 0 0>}]
 dyn_syrup_D_02 [{shapeType=Capsule, length=0.9062, radius=2.5, center=<2 -1 0>, rotation=<rot 90 0 0>}]
 dyn_syrup_E_01 [{shapeType=Capsule, length=3.125, radius=1, center=<2.5481 0 0>, rotation=<rot 90 0 0>}]
 dyn_syrup_E_02 [{shapeType=Capsule, length=3.1875, radius=1, center=<2.6939 0 0>, rotation=<rot 90 0 0>}]
 dyn_syrup_F_01 [{shapeType=Capsule, length=3.125, radius=1, center=<2.4616 0 0.2345>, rotation=<rot 90 0 0>}]
 dyn_syrup_F_02 [{shapeType=Capsule, length=2.625, radius=1.5, center=<2.5793 0 1>, rotation=<rot 90 0 0>}]
 dyn_syrup_F_03 [{shapeType=Capsule, length=1.7188, radius=1.7812, center=<2 0 1.9088>, rotation=<rot 90 0 0>}]
 dyn_syrup_G_01 [{shapeType=Capsule, length=5.125, radius=1, center=<3.1924 0 0>, rotation=<rot 90 0 0>}]
 dyn_syrup_G_02 [{shapeType=Capsule, length=4.3125, radius=1, center=<2.9027 0 0>, rotation=<rot 90 0 0>}]
 dyn_syrup_G_03 [{shapeType=Capsule, length=4.3125, radius=1, center=<3.0755 -0.0001 0>, rotation=<rot 90 0 0>}]
 dyn_syrup_G_04 [{shapeType=Capsule, length=3.6562, radius=1.6562, center=<1.8 0 0>, rotation=<rot 90 0 0>}]
 dyn_syrup_H_01 [{shapeType=Capsule, length=3.2812, radius=1, center=<2.1696 0 0>, rotation=<rot 90 0 0>}]
 dyn_syrup_H_02 [{shapeType=Capsule, length=3.2812, radius=1, center=<2.2026 0 0>, rotation=<rot 90 0 0>}]
 dyn_syrup_H_03 [{shapeType=Capsule, length=1.5625, radius=1.4375, center=<1.8 0 0>, rotation=<rot 90 0 0>}]
 dyn_syrup_I_01 [{shapeType=Capsule, length=3.5, radius=1, center=<2.7495 0.0001 0>, rotation=<rot 90 0 0>}]
 dyn_syrup_I_02 [{shapeType=Capsule, length=3.7812, radius=1, center=<2.5693 0 0>, rotation=<rot 90 0 0>}]
 dyn_syrup_I_03 [{shapeType=Capsule, length=3.3125, radius=1.4062, center=<2.3825 0 0>, rotation=<rot 90 0 0>}]
 dyn_syrup_J_01 [{shapeType=Capsule, length=2.7812, radius=1, center=<1.8 0 0>, rotation=<rot 90 0 0>}]
 dyn_syrup_J_02 [{shapeType=Capsule, length=2, radius=1, center=<1.8 0 0>, rotation=<rot 90 0 0>}]
 head [{shapeType=Capsule, length=3.3125, radius=10.5625, center=<0 1.6704 -0.5308>, rotation=<rot 90 0 0>}]
 neck_01 [{shapeType=Capsule, length=3.3125, radius=10.5625, center=<0 3.5487 -0.531>, rotation=<rot 90 0 0>}]
 neck_02 [{shapeType=Capsule, length=3.3125, radius=10.5625, center=<0 2.8987 -0.5308>, rotation=<rot 90 0 10.0003>}]
 spine_05 [
  {shapeType=Capsule, length=16.6875, radius=11.3438, center=<0 3.5487 -0.531>}
  {shapeType=Capsule, length=11.8125, radius=4.75, center=<7.9791 6.0852 11.6936>, rotation=<rot -68.038 -16.1295 -30.1659>}
  {shapeType=Capsule, length=16.0625, radius=4.75, center=<8.4354 5.7656 -11.3261>, rotation=<rot -72.6217 78.4378 -124.8438>}
  {shapeType=Capsule, length=10.2812, radius=7.9375, center=<13.2167 -2.4436 -0.531>}
 ]
 upperarm_r [{shapeType=Capsule, length=21.25, radius=5, shapeName=upperarm_r_capsule, center=<-15.5734 0 0>, rotation=<rot 90 0 0>}]
}
constraints [
 {
  bone1Name=dyn_pancake_B_side_02
  bone2Name=dyn_pancake_base_B
  swing1LimitDegrees=0.1
  swing2LimitDegrees=10
  twistLimitDegrees=5
  twistMotion=Locked
 }
 {
  bone1Name=dyn_pancake_C_side_02
  bone2Name=dyn_pancake_base_C
  swing1LimitDegrees=0.1
  swing2LimitDegrees=10
  twistLimitDegrees=5
  twistMotion=Locked
 }
 {
  bone1Name=dyn_pancake_D_side_01
  bone2Name=dyn_pancake_base_D
  swing1LimitDegrees=0.1
  swing2LimitDegrees=10
  twistLimitDegrees=5
  twistMotion=Locked
 }
 {
  bone1Name=dyn_pancake_E_side_02
  bone2Name=dyn_pancake_base_E
  swing1LimitDegrees=0.1
  swing2LimitDegrees=10
  twistLimitDegrees=5
  twistMotion=Locked
 }
 {
  bone1Name=dyn_pancake_F_side_01
  bone2Name=dyn_pancake_base_F
  swing1LimitDegrees=0.1
  swing2LimitDegrees=10
  twistLimitDegrees=5
  twistMotion=Locked
 }
 {
  bone1Name=dyn_pancake_F_side_02
  bone2Name=dyn_pancake_base_F
  swing1LimitDegrees=0.1
  swing2LimitDegrees=10
  twistLimitDegrees=5
  twistMotion=Locked
 }
 {
  bone1Name=dyn_pancake_G_side_01
  bone2Name=dyn_pancake_base_G
  swing1LimitDegrees=0.1
  swing2LimitDegrees=10
  twistLimitDegrees=5
  twistMotion=Locked
 }
 {bone1Name=dyn_syrup_A_01, bone2Name=dyn_pancake_G_side_01, swing1LimitDegrees=30, swing2LimitDegrees=30, twistLimitDegrees=5}
 {bone1Name=dyn_syrup_J_01, bone2Name=dyn_pancake_G_side_01, swing1LimitDegrees=10, swing2LimitDegrees=10, twistLimitDegrees=5}
 {bone1Name=dyn_syrup_J_02, bone2Name=dyn_syrup_J_01, swing1LimitDegrees=10, swing2LimitDegrees=10, twistLimitDegrees=5}
 {bone1Name=dyn_syrup_B_01, bone2Name=dyn_pancake_G_side_01, swing1LimitDegrees=30, swing2LimitDegrees=30, twistLimitDegrees=5}
 {bone1Name=dyn_syrup_B_02, bone2Name=dyn_syrup_B_01, swing1LimitDegrees=30, swing2LimitDegrees=30, twistLimitDegrees=5}
 {
  bone1Name=dyn_pancake_G_side_02
  bone2Name=dyn_pancake_base_G
  swing1LimitDegrees=0.1
  swing2LimitDegrees=10
  twistLimitDegrees=5
  twistMotion=Locked
 }
 {bone1Name=dyn_syrup_D_01, bone2Name=dyn_pancake_G_side_02, swing1LimitDegrees=30, swing2LimitDegrees=30, twistLimitDegrees=5}
 {bone1Name=dyn_syrup_D_02, bone2Name=dyn_syrup_D_01, swing1LimitDegrees=30, swing2LimitDegrees=30, twistLimitDegrees=5}
 {bone1Name=dyn_syrup_E_01, bone2Name=dyn_pancake_G_side_02, swing1LimitDegrees=30, swing2LimitDegrees=30, twistLimitDegrees=5}
 {bone1Name=dyn_syrup_E_02, bone2Name=dyn_syrup_E_01, swing1LimitDegrees=30, swing2LimitDegrees=30, twistLimitDegrees=5}
 {bone1Name=dyn_syrup_F_01, bone2Name=dyn_pancake_G_side_02, swing1LimitDegrees=10, swing2LimitDegrees=10, twistLimitDegrees=5}
 {bone1Name=dyn_syrup_F_02, bone2Name=dyn_syrup_F_01, swing1LimitDegrees=10, swing2LimitDegrees=10, twistLimitDegrees=5}
 {bone1Name=dyn_syrup_F_03, bone2Name=dyn_syrup_F_02, swing1LimitDegrees=10, swing2LimitDegrees=10, twistLimitDegrees=5}
 {bone1Name=dyn_syrup_G_01, bone2Name=dyn_pancake_G_side_02, swing1LimitDegrees=10, swing2LimitDegrees=10, twistLimitDegrees=5}
 {bone1Name=dyn_syrup_G_02, bone2Name=dyn_syrup_G_01, swing1LimitDegrees=10, swing2LimitDegrees=30, twistLimitDegrees=5}
 {bone1Name=dyn_syrup_G_03, bone2Name=dyn_syrup_G_02, swing1LimitDegrees=10, swing2LimitDegrees=30, twistLimitDegrees=5}
 {bone1Name=dyn_syrup_G_04, bone2Name=dyn_syrup_G_03, swing1LimitDegrees=10, swing2LimitDegrees=30, twistLimitDegrees=5}
 {bone1Name=dyn_syrup_H_01, bone2Name=dyn_pancake_G_side_02, swing1LimitDegrees=10, swing2LimitDegrees=5, twistLimitDegrees=5}
 {bone1Name=dyn_syrup_H_02, bone2Name=dyn_syrup_H_01, swing1LimitDegrees=30, swing2LimitDegrees=30, twistLimitDegrees=5}
 {bone1Name=dyn_syrup_H_03, bone2Name=dyn_syrup_H_02, swing1LimitDegrees=30, swing2LimitDegrees=30, twistLimitDegrees=5}
 {bone1Name=dyn_syrup_C_01, bone2Name=dyn_pancake_base_G, swing1LimitDegrees=30, swing2LimitDegrees=30, twistLimitDegrees=5}
 {bone1Name=dyn_syrup_C_02, bone2Name=dyn_syrup_C_01, swing1LimitDegrees=30, swing2LimitDegrees=30, twistLimitDegrees=5}
 {bone1Name=dyn_syrup_I_01, bone2Name=dyn_pancake_base_G, swing1LimitDegrees=30, swing2LimitDegrees=30, twistLimitDegrees=5}
 {bone1Name=dyn_syrup_I_02, bone2Name=dyn_syrup_I_01, swing1LimitDegrees=30, swing2LimitDegrees=30, twistLimitDegrees=5}
 {bone1Name=dyn_syrup_I_03, bone2Name=dyn_syrup_I_02, swing1LimitDegrees=30, swing2LimitDegrees=30, twistLimitDegrees=5}
 {bone1Name=dyn_butter, bone2Name=dyn_pancake_base_G, swing1LimitDegrees=2, swing2LimitDegrees=2, twistLimitDegrees=2}
 {
  bone1Name=dyn_pancake_base_E
  bone2Name=head
  swing1LimitDegrees=30
  swing1Motion=Locked
  swing2LimitDegrees=30
  swing2Motion=Locked
  twistLimitDegrees=5
  twistMotion=Locked
 }
 {
  bone1Name=dyn_pancake_base_F
  bone2Name=head
  swing1LimitDegrees=30
  swing1Motion=Locked
  swing2LimitDegrees=30
  swing2Motion=Locked
  twistLimitDegrees=5
  twistMotion=Locked
 }
 {
  bone1Name=dyn_pancake_base_G
  bone2Name=head
  swing1LimitDegrees=30
  swing1Motion=Locked
  swing2LimitDegrees=30
  swing2Motion=Locked
  twistLimitDegrees=5
  twistMotion=Locked
 }
 {
  bone1Name=head
  bone2Name=neck_02
  swing1LimitDegrees=45
  swing1Motion=Free
  swing2LimitDegrees=45
  swing2Motion=Free
  twistLimitDegrees=45
  twistMotion=Free
 }
 {
  bone1Name=dyn_pancake_base_C
  bone2Name=neck_02
  swing1LimitDegrees=30
  swing1Motion=Locked
  swing2LimitDegrees=30
  swing2Motion=Locked
  twistLimitDegrees=5
  twistMotion=Locked
 }
 {
  bone1Name=dyn_pancake_base_D
  bone2Name=neck_02
  swing1LimitDegrees=30
  swing1Motion=Locked
  swing2LimitDegrees=30
  swing2Motion=Locked
  twistLimitDegrees=5
  twistMotion=Locked
 }
 {
  bone1Name=neck_02
  bone2Name=neck_01
  swing1LimitDegrees=45
  swing1Motion=Free
  swing2LimitDegrees=45
  swing2Motion=Free
  twistLimitDegrees=45
  twistMotion=Free
 }
 {
  bone1Name=dyn_pancake_base_B
  bone2Name=neck_01
  swing1LimitDegrees=30
  swing1Motion=Locked
  swing2LimitDegrees=30
  swing2Motion=Locked
  twistLimitDegrees=5
  twistMotion=Locked
 }
 {
  bone1Name=neck_01
  bone2Name=spine_05
  swing1LimitDegrees=45
  swing1Motion=Free
  swing2LimitDegrees=45
  swing2Motion=Free
  twistLimitDegrees=45
  twistMotion=Free
 }
 {
  bone1Name=dyn_pancake_base_A
  bone2Name=spine_05
  swing1LimitDegrees=45
  swing1Motion=Free
  swing2LimitDegrees=45
  swing2Motion=Free
  twistLimitDegrees=45
  twistMotion=Free
 }
 {
  bone1Name=upperarm_r
  bone2Name=spine_05
  swing1LimitDegrees=45
  swing1Motion=Free
  swing2LimitDegrees=45
  swing2Motion=Free
  twistLimitDegrees=45
  twistMotion=Free
 }
]
