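# Post-process-style Anim Blueprint for Flappy's face accessory mesh, animating syrup strands and melting butter.
# Each update derives neck and head delta rotations, head up/down alphas and a ten-minute world-time butter melt.
# Copied poses are layered, partial rotations drive ALT neck/head bones, syrup gets constraints and RigidBody simulation, then additive and replace correctives apply.
# A Control Rig node and an earlier syrup-rotation chain sit unconnected to the output.
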AnimBlueprint /BRCosmetics/Characters/Player/Male/Medium/Bodies/M_MED_Flappy/Meshes/Parts/M_MED_Flappy_FaceAcc_AnimBP
target_skeleton=&/M_MED_Flappy_FaceAcc_Skeleton
graphs [
 AnimGraph {nodes [
   Control Rig() {
    alphaBoolBlend {blendOption=Linear}
    controlRigAssetReference {blueprintRigClass=&/M_MED_Flappy_FaceAcc_ControlRig.M_MED_Flappy_FaceAcc_ControlRig_C}
   }
   n1 = Get headDownFloat() -> (headDownFloat:double)
   n2 = "Transform (Modify) Bone - Bone: dyn_syrup_G_01"(Alpha=@n1, Translation="0,0,0", Rotation="-5,0.000,0.000", Scale="1,1,1") {
    rotationMode=BMM_Additive
    rotationSpace=BCS_BoneSpace
    scaleSpace=BCS_BoneSpace
    translationMode=BMM_Additive
    translationSpace=BCS_BoneSpace
    alphaBoolBlend {blendOption=Linear}
    boneToModify {boneName=dyn_syrup_G_01}
    rotation {pitch=-5}
   }
   n3 = "Transform (Modify) Bone - Bone: dyn_syrup_G_02"(ComponentPose=@n2, Alpha=@n1, Translation="0,0,0", Rotation="5,0.000,0.000", Scale="1,1,1") {
    rotationMode=BMM_Additive
    rotationSpace=BCS_BoneSpace
    scaleSpace=BCS_BoneSpace
    translationMode=BMM_Additive
    translationSpace=BCS_BoneSpace
    alphaBoolBlend {blendOption=Linear}
    boneToModify {boneName=dyn_syrup_G_02}
    rotation {pitch=5}
   }
   n4 = "Transform (Modify) Bone - Bone: dyn_syrup_H_01"(ComponentPose=@n3, Alpha=@n1, Translation="0,0,0", Rotation="-15,0.000,0.000", Scale="1,1,1") {
    rotationMode=BMM_Additive
    rotationSpace=BCS_BoneSpace
    scaleSpace=BCS_BoneSpace
    translationMode=BMM_Additive
    translationSpace=BCS_BoneSpace
    alphaBoolBlend {blendOption=Linear}
    boneToModify {boneName=dyn_syrup_H_01}
    rotation {pitch=-15}
   }
   n5 = "Transform (Modify) Bone - Bone: dyn_syrup_H_02"(ComponentPose=@n4, Alpha=@n1, Translation="0,0,0", Rotation="15,0.000,0.000", Scale="1,1,1") {
    rotationMode=BMM_Additive
    rotationSpace=BCS_BoneSpace
    scaleSpace=BCS_BoneSpace
    translationMode=BMM_Additive
    translationSpace=BCS_BoneSpace
    alphaBoolBlend {blendOption=Linear}
    boneToModify {boneName=dyn_syrup_H_02}
    rotation {pitch=15}
   }
   "Transform (Modify) Bone - Bone: dyn_syrup_H_03"(ComponentPose=@n5, Alpha=@n1, Translation="0,0,0", Rotation="15,0.000,0.000", Scale="1,1,1") {
    rotationMode=BMM_Additive
    rotationSpace=BCS_BoneSpace
    scaleSpace=BCS_BoneSpace
    translationMode=BMM_Additive
    translationSpace=BCS_BoneSpace
    alphaBoolBlend {blendOption=Linear}
    boneToModify {boneName=dyn_syrup_H_03}
    rotation {pitch=15}
   }
   n6 = Copy Pose From Mesh(SourceMeshComponent=Get Skel_Mesh()) {bCopyCurves=true, bUseAttachedParent=true}
   n7 = Copy Pose From Mesh() {bCopyCurves=true, bUseAttachedParent=true}
   n8 = Get Neck 01 Rot() -> (Neck 01 Rot:Rotator)
   n9 = "Transform (Modify) Bone - Bone: neck_01_ALT_A"(
    ComponentPose=Local To Component(LocalPose=Layered blend per bone(BasePose=@n7, BlendPoses_0=@n6) {layerSetup [{branchFilters [{boneName=upperarm_r}]}]})
    Alpha=0.5
    Translation="0,0,0"
    Rotation=@n8
    Scale="1,1,1"
   ) {rotationMode=BMM_Replace, rotationSpace=BCS_BoneSpace, alphaBoolBlend {blendOption=Linear}, boneToModify {boneName=neck_01_ALT_A}}
   n10 = "Transform (Modify) Bone - Bone: neck_01_ALT_B"(ComponentPose=@n9, Alpha=0.5, Translation="0,0,0", Rotation=@n8, Scale="1,1,1") {rotationMode=BMM_Replace, rotationSpace=BCS_BoneSpace, alphaBoolBlend {blendOption=Linear}, boneToModify {boneName=neck_01_ALT_B}}
   n11 = Get Neck 02 Rot() -> (Neck 02 Rot:Rotator)
   n12 = "Transform (Modify) Bone - Bone: neck_02_ALT_C"(ComponentPose=@n10, Alpha=0.5, Translation="0,0,0", Rotation=@n11, Scale="1,1,1") {rotationMode=BMM_Replace, rotationSpace=BCS_BoneSpace, alphaBoolBlend {blendOption=Linear}, boneToModify {boneName=neck_02_ALT_C}}
   n13 = "Transform (Modify) Bone - Bone: neck_02_ALT_D"(ComponentPose=@n12, Alpha=0.5, Translation="0,0,0", Rotation=@n11, Scale="1,1,1") {rotationMode=BMM_Replace, rotationSpace=BCS_BoneSpace, alphaBoolBlend {blendOption=Linear}, boneToModify {boneName=neck_02_ALT_D}}
   n14 = Get head rot() -> (head rot:Rotator)
   n15 = "Transform (Modify) Bone - Bone: head_ALT_E"(ComponentPose=@n13, Alpha=0.15, Translation="0,0,0", Rotation=@n14, Scale="1,1,1") {rotationMode=BMM_Replace, rotationSpace=BCS_BoneSpace, alphaBoolBlend {blendOption=Linear}, boneToModify {boneName=head_ALT_E}}
   n16 = "Transform (Modify) Bone - Bone: head_ALT_F"(ComponentPose=@n15, Alpha=0.333, Translation="0,0,0", Rotation=@n14, Scale="1,1,1") {rotationMode=BMM_Replace, rotationSpace=BCS_BoneSpace, alphaBoolBlend {blendOption=Linear}, boneToModify {boneName=head_ALT_F}}
   n17 = "Transform (Modify) Bone - Bone: head_ALT_G"(ComponentPose=@n16, Alpha=0.333, Translation="0,0,0", Rotation=@n14, Scale="1,1,1") {rotationMode=BMM_Replace, rotationSpace=BCS_BoneSpace, alphaBoolBlend {blendOption=Linear}, boneToModify {boneName=head_ALT_G}}
   n18 = "Constraint - dyn_syrup_G_01 (P)"(ComponentPose=@n17, Alpha=1, ConstraintWeights_0=1) {
    alphaBoolBlend {blendOption=Linear}
    boneToModify {boneName=dyn_syrup_G_01}
    constraintSetup [{targetBone {boneName=head_ALT_F}}]
    constraintWeights [1]
   }
   n19 = "Constraint - dyn_syrup_H_01 (P)"(ComponentPose=@n18, Alpha=1, ConstraintWeights_0=1) {
    alphaBoolBlend {blendOption=Linear}
    boneToModify {boneName=dyn_syrup_H_01}
    constraintSetup [{targetBone {boneName=head_ALT_F}}]
    constraintWeights [1]
   }
   n20 = "Transform (Modify) Bone - Bone: dyn_syrup_H_01"(ComponentPose=@n19, Alpha=@n1, Translation="0,0,0", Rotation="10,0.000,0.000", Scale="1,1,1") {
    rotationMode=BMM_Additive
    rotationSpace=BCS_BoneSpace
    scaleSpace=BCS_BoneSpace
    translationMode=BMM_Additive
    translationSpace=BCS_BoneSpace
    alphaBoolBlend {blendOption=Linear}
    boneToModify {boneName=dyn_syrup_H_01}
    rotation {pitch=10}
   }
   n21 = RigidBody(ComponentPose=@n20) {
    bClampLinearTranslationLimitToRefPose=true
    bOverrideWorldGravity=true
    overridePhysicsAsset=&/M_MED_Flappy_FaceAcc_Dyn
    simulationSpace=BaseBoneSpace
    alphaBoolBlend {blendOption=Linear}
    baseBoneRef {boneName=head}
    componentAppliedLinearAccClamp=<3000 3000 3000>
    componentLinearAccScale=<0.35 0.3 0.3>
    componentLinearVelScale=<0.35 0.3 0.3>
    overrideWorldGravity {z=-980}
   }
   n22 = "Transform (Modify) Bone - Bone: dyn_syrup_C_01"(ComponentPose=@n21, Alpha=Get headUpFloat() -> (headUpFloat:double), Translation="0,0,0", Rotation="10,0,0", Scale="1,1,1") {
    rotationMode=BMM_Replace
    rotationSpace=BCS_BoneSpace
    alphaBoolBlend {blendOption=Linear}
    boneToModify {boneName=dyn_syrup_C_01}
    rotation {pitch=10}
   }
   n23 = "Transform (Modify) Bone - Bone: dyn_syrup_D_01"(ComponentPose=@n22, Alpha=Get headUpFloat(), Translation="0,0,0", Rotation="25,0,0", Scale="1,1,1") {
    rotationMode=BMM_Replace
    rotationSpace=BCS_BoneSpace
    alphaBoolBlend {blendOption=Linear}
    boneToModify {boneName=dyn_syrup_D_01}
    rotation {pitch=25}
   }
   n24 = "Transform (Modify) Bone - Bone: dyn_butter"(
    ComponentPose=@n23
    Alpha=Get butterMeltFloat() -> (butterMeltFloat:double)
    Translation="-0.5,0,0"
    Rotation="0,0,0"
    Scale="0.05,0.05,0.05"
   ) {
    rotationSpace=BCS_BoneSpace
    scaleMode=BMM_Replace
    scaleSpace=BCS_BoneSpace
    translationMode=BMM_Additive
    translationSpace=BCS_BoneSpace
    alphaBoolBlend {blendOption=Linear}
    boneToModify {boneName=dyn_butter}
    scale=<0.05 0.05 0.05>
    translation {x=-0.5}
   }
   n25 = "Transform (Modify) Bone - Bone: dyn_syrup_F_01"(ComponentPose=@n24, Alpha=1, Translation="0,0,0.5", Rotation="0,0,0", Scale="1,1,1") {
    rotationMode=BMM_Additive
    rotationSpace=BCS_BoneSpace
    scaleSpace=BCS_BoneSpace
    translationMode=BMM_Additive
    translationSpace=BCS_BoneSpace
    alphaBoolBlend {blendOption=Linear}
    boneToModify {boneName=dyn_syrup_F_01}
    translation {z=0.5}
   }
   n26 = "Transform (Modify) Bone - Bone: dyn_syrup_G_01"(ComponentPose=@n25, Alpha=@n1, Translation="0,0,0", Rotation="5,0,0.000", Scale="1,1,1") {
    rotationMode=BMM_Additive
    rotationSpace=BCS_BoneSpace
    scaleSpace=BCS_BoneSpace
    translationMode=BMM_Additive
    translationSpace=BCS_BoneSpace
    alphaBoolBlend {blendOption=Linear}
    boneToModify {boneName=dyn_syrup_G_01}
    rotation {pitch=5}
   }
   Output Pose(Result=Component To Local(ComponentPose=@n26)) {layerGroup=None}
  ]}
 EventGraph {nodes [
   Event BlueprintUpdateAnimation()
   n1 = Get Skel_Mesh()
   n2 = BreakTransform(InTransform=GetDeltaTransformFromRefPose(self=@n1, BoneName=head, BaseName=neck_02))
   > Set head rot(head rot:Rotator=@n2.Rotation) -> (Output_Get:Rotator)
   > Set Neck 01 Rot(Neck 01 Rot:Rotator=BreakTransform(InTransform=GetDeltaTransformFromRefPose(self=@n1, BoneName=neck_01, BaseName=spine_05)).Rotation) -> (Output_Get:Rotator)
   > Set Neck 02 Rot(Neck 02 Rot:Rotator=BreakTransform(InTransform=GetDeltaTransformFromRefPose(self=@n1, BoneName=neck_02, BaseName=neck_01)).Rotation) -> (Output_Get:Rotator)
   > Set butterMeltFloat(butterMeltFloat:double=MapRangeClamped(Value=GetTimeSeconds!() -> (ReturnValue:double), InRangeA=0, InRangeB=600)) -> (Output_Get:double)
   n3 = BreakTransform(InTransform=GetDeltaTransformFromRefPose(self=Get Skel_Mesh(), BoneName=head, BaseName=neck_02))
   > Set headDownFloat(headDownFloat:double=MapRangeClamped(Value=BreakRotator(InRot=@n3.Rotation).Yaw, InRangeA=0, InRangeB=20)) -> (Output_Get:double)
   > Set headUpFloat(headUpFloat:double=MapRangeClamped(Value=BreakRotator(InRot=@n2.Rotation).Yaw, InRangeA=0, InRangeB=-30)) -> (Output_Get:double)
   Event BlueprintInitializeAnimation()
   > Set Skel_Mesh(Skel_Mesh=GetOwnerSkeletalMeshComponent())
  ]}
]
variables [
 butterMeltFloat {type=number}
 head rot {type=Rotator}
 headDownFloat {type=number}
 headUpFloat {type=number}
 Neck 01 Rot {type=Rotator}
 Neck 02 Rot {type=Rotator}
 Skel_Mesh {type=SkeletalMeshComponent}
]
